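# Physics asset for the in-hand Boom Box gadget skeletal mesh.
# It defines a single body on the Boom_Box_In_Hand bone, using one near-cubic box shape centered close to the bone origin and rotated about a quarter turn.
# It has no constraints, so there are no jointed parts to simulate.
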PhysicsAsset /Game/Gadgets/Assets/Boom_Box/Meshes/SK_BoomBox_InHand_PhysicsAsset
bodies {Boom_Box_In_Hand [{shapeType=Box, extentX=11.1958, extentY=11.1934, extentZ=11.844, center=<-0.1816 -0.0538 0.1229>, rotation=<rot 0.0819 -89.9995 0.2667>}]}
constraints []
